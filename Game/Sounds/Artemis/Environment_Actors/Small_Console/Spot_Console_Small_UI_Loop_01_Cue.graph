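# Sound Cue for the looping UI sound of a small console environment actor, routed to the gadgets sound class with its own attenuation.
# A looping wave player feeds an equally weighted Random node with an empty second slot, then a short randomized Delay and an Enveloper that eases volume in.
# The first two quality-level slots share this chain and the other two are empty.
# A concurrency override tightly limits simultaneous instances.
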SoundCue /Game/Sounds/Artemis/Environment_Actors/Small_Console/Spot_Console_Small_UI_Loop_01_Cue
root=QualityLevel_0
nodes [
 {id=Delay_0, children [Random_0], properties {delayMax=0.1, delayMin=0.05}}
 {
  id=Enveloper_29
  children [Delay_0]
  properties {
   pitchCurve {editorCurveData {keys [{time=0}]}}
   volumeCurve {editorCurveData {keys [{arriveTangent=3.4293, interpMode=RCIM_Cubic, leaveTangent=3.4293, tangentMode=RCTM_User, time=0, value=0}, <at 0.4 1 RCIM_Cubic>]}}
  }
 }
 {id=QualityLevel_0, children [Enveloper_29,Enveloper_29,null,null]}
 {id=Random_0, children [WavePlayer_2,null], properties {weights [1,1]}}
 {
  id=WavePlayer_2
  properties {bLooping=true, soundWaveAssetPtr=&/Spot_Console_Small_UI_Loop_01}
 }
]
properties {
 attenuationSettings=&/Spot_Console_Small_UI_Att
 bEnableBusSends=false
 bEnableSubmixSends=false
 bOverrideConcurrency=true
 maxDistance=1000
 soundClassObject=&Fort_Gadgets
 volumeMultiplier=0.66
 concurrencyOverrides {maxCount=2, voiceStealReleaseTime=0.2}
}
