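# Looping sound cue for the static hum of a small console screen, played quietly as a spatial environment spot sound.
# A mixer layers one of four equally weighted hologram loops, pitched down to half, with a computer transmitter loop at a randomized lowered pitch.
# It uses a dedicated attenuation asset and the gadgets and items sound class, with a concurrency override that prevents new instances at the limit.
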SoundCue /Game/Sounds/Artemis/Environment_Actors/Small_Console/Spot_Small_Screen_Static_Hum_Loop_01_Cue
root=Mixer_0
nodes [
 {id=Mixer_0, children [Modulator_0,Modulator_6], properties {inputVolume [0.75,0.75]}}
 {id=Modulator_0, children [Random_0], properties {pitchMax=0.5, pitchMin=0.5}}
 {id=Modulator_6, children [WavePlayer_0], properties {pitchMax=0.7, pitchMin=0.5}}
 {id=Random_0, children [WavePlayer_1,WavePlayer_2,WavePlayer_3,WavePlayer_4], properties {weights [1,1,1,1]}}
 {
  id=WavePlayer_0
  properties {bLooping=true, soundWaveAssetPtr=/Game/Athena/Sounds/Props/MountainBaseComputer/ComputerTransimtter_Running_Loop_01}
 }
 {
  id=WavePlayer_1
  properties {bLooping=true, soundWaveAssetPtr=/Game/Sounds/Foley_Loot/Containers/Keys_And_Locks/Keys_And_Locks_HologramLoop_01}
 }
 {
  id=WavePlayer_2
  properties {bLooping=true, soundWaveAssetPtr=/Game/Sounds/Foley_Loot/Containers/Keys_And_Locks/Keys_And_Locks_HologramLoop_02}
 }
 {
  id=WavePlayer_3
  properties {bLooping=true, soundWaveAssetPtr=/Game/Sounds/Foley_Loot/Containers/Keys_And_Locks/Keys_And_Locks_HologramLoop_03}
 }
 {
  id=WavePlayer_4
  properties {bLooping=true, soundWaveAssetPtr=/Game/Sounds/Foley_Loot/Containers/Keys_And_Locks/Keys_And_Locks_HologramLoop_04}
 }
]
properties {
 attenuationSettings=&/Spot_Small_Screen_Static_Loop_Att
 bOverrideConcurrency=true
 maxDistance=800
 soundClassObject=&Fort_Gadgets_Items
 volumeMultiplier=0.25
 attenuationOverrides {falloffDistance=2000}
 concurrencyOverrides {resolutionRule=PreventNew}
 concurrencySet [&Weapon_Drop_StopOldest]
}
