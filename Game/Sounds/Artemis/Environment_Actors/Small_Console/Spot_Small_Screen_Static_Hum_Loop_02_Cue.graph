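# Sound cue for a small screen's ambient static hum, layering a steady electronic bed with intermittent computer blips.
# Two looping waves are mixed and pitched down by half, offsetting the cue's doubled pitch multiplier.
# Alongside, a looped branch waits a random delay, then plays one of two computer one-shots with slight pitch variation.
# It uses spatial attenuation and overrides concurrency to prevent new instances at the limit.
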SoundCue /Game/Sounds/Artemis/Environment_Actors/Small_Console/Spot_Small_Screen_Static_Hum_Loop_02_Cue
root=Mixer_0
nodes [
 {id=Delay_1, children [Modulator_2], properties {delayMax=3, delayMin=0.5}}
 {id=Looping_1, children [Delay_1]}
 {id=Mixer_0, children [Modulator_1,Looping_1], properties {inputVolume [0.6,0.25]}}
 {id=Mixer_1, children [WavePlayer_0,WavePlayer_6], properties {inputVolume [0.5,0.2]}}
 {id=Modulator_1, children [Mixer_1], properties {pitchMax=0.5, pitchMin=0.5}}
 {id=Modulator_2, children [Random_2], properties {pitchMax=1.15, pitchMin=0.9, volumeMax=1, volumeMin=1}}
 {id=Random_2, children [WavePlayer_5, WavePlayer_10], properties {weights [1,1]}}
 {id=WavePlayer_0, properties {bLooping=true, soundWaveAssetPtr=/Game/Sounds/Creative/Toys/Locks/LockDevice_Screen_Loop_02}}
 {id=WavePlayer_10, properties {soundWaveAssetPtr=/Game/Sounds/Apollo/Biomes/GovernmentSecretComplex/SecretComplex_Computer_Oneshot_07}}
 {id=WavePlayer_5, properties {soundWaveAssetPtr=/Game/Sounds/Apollo/Biomes/GovernmentSecretComplex/SecretComplex_Computer_Oneshot_01}}
 {
  id=WavePlayer_6
  properties {bLooping=true, soundWaveAssetPtr=&/Spot_Console_Small_UI_Loop_01}
 }
]
properties {
 attenuationSettings=&/Spot_Small_Screen_Static_Loop_Att
 bOverrideConcurrency=true
 maxDistance=800
 pitchMultiplier=2
 soundClassObject=&Fort_Gadgets_Items
 volumeMultiplier=0.25
 attenuationOverrides {falloffDistance=2000}
 concurrencyOverrides {resolutionRule=PreventNew}
 concurrencySet [&Weapon_Drop_StopOldest]
}
